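# Niagara idle effect for the EssayViewMyth backpack cosmetic, whose emitter names point to a hamster wheel powering a glowing capsule.
# System.ReactivityIntensity scales glows, two lights, wheel and hamster speedlines, motion blur, lightning and sparks.
# A socket reader anchors the capsule, wheel and sparks, while lightning picks mesh, placement and orientation from indexed arrays.
# Front-end modifiers and time-of-day functions adjust brightness, tint and camera offsets on some emitters.
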NiagaraSystem /BRCosmetics/Accessories/FORT_Backpacks/Backpack_M_MED_EssayViewMyth/FX/NS_Backpack_EssayViewMyth_Idle
dependencies {
 usedDataInterfaces [
  &NiagaraDataInterfaceMeshRendererInfo
  &NiagaraDataInterfaceCurve
  &NiagaraDataInterfaceArrayInt32
  &NiagaraDataInterfaceArrayFloat3
  &NiagaraDataInterfaceArrayFloat
  &NiagaraDataInterfaceColorCurve
  &NiagaraDataInterfaceSocketReader
 ]
 usedDynamicInputs [
  &Add_Float
  &Multiply_FloatByInt
  &SpawnGroup
  &Multiply_Vector
  &SimulationPosition
  &Add_Integer
  &Multiply_Int
  &Subtract_Float
 ]
 usedModules [
  &self:SystemSpawnScript.SystemScriptSource.SystemScriptGraph.NiagaraNodeAssignment_4.SetVariables_#1
  &self:SystemSpawnScript.SystemScriptSource.SystemScriptGraph.NiagaraNodeAssignment_1.SetVariables_#2
  &SystemState
  &self:SystemSpawnScript.SystemScriptSource.SystemScriptGraph.NiagaraNodeAssignment_3.SetVariables_#3
 ]
}
emitters {
 Capsule_Glow_And_Light {
  emitterSpawnScript [/SetVariables(Emitter.CapsuleGlowDirection:Vector3f=<1 0 0>) {isSetParameters=true}]
  emitterUpdateScript [
   &EmitterState(
    Inactive Response=::NewEnumerator1
    Life Cycle Mode=::NewEnumerator1
    Loop Behavior=::NewEnumerator1
    Loop Duration Mode=::NewEnumerator1
    UseLoopDelay=false
   )
   &SpawnBurst_Instantaneous(Use Spawn Probability=false, UseLoopCountLimit=false)
  ]
  particleSpawnScript [
   &InitializeParticle(
    Color=$System.Glow Color
    Position Offset=<0 -5.5 13>
    Position Offset Coordinate Space=::Local
    Sprite Size Mode=::NewEnumerator3
    Uniform Sprite Size=20
    UsePositionOffset=true
    Write Material Random=true
   )
   &CameraOffset()
  ]
  particleUpdateScript [
   &ParticleState(Kill Particles When Lifetime Has Elapsed=false, Loop Particles Lifetime=true)
   &SocketLocation(
    SamplePosOffset=true
    Sampled Position Offset=<0 -2 0>
    Socket Reader=$System.Socket Reader
    Socket Sampling Mode=::NewEnumerator3
    VelocitySpeedLimit=false
   )
   &SocketLocation(
    Orientation Sampling=::NewEnumerator1
    Position Sampling=::NewEnumerator1
    SamplePosOffset=true
    Sampled Position Offset=<0 -5 0>
    Socket Index=1
    Socket Reader=$System.Socket Reader
    Socket Sampling Mode=::NewEnumerator3
    VelocitySpeedLimit=false
   ) {module=SocketLocation001}
   &ScaleColor(
    Scale Alpha=&FloatFromCurve(
     FloatCurve=DataInterface() {propertyValues {Curve {keys [
         <at 0 0.7 RCIM_Cubic>
         {arriveTangent=0.4167, interpMode=RCIM_Cubic, leaveTangent=0.4167, time=0.3, value=1}
         {arriveTangent=0.8965, interpMode=RCIM_Cubic, leaveTangent=0.8965, tangentMode=RCTM_User, time=0.6, value=0.95}
         <at 1 0.7 RCIM_Cubic>
        ]}}}
     Scale Curve=&Lerp_Float(A=0, B=&FN_FrontEndFloatModifier(Gameplay=1, Front End=0.5), Alpha=$System.ReactivityIntensity)
    )
    Scale RGB=&VectorFromFloat(Value=&Lerp_Float(A=1, B=1, Alpha=$User.ReactivityTime))
    ScaleA=true
    ScaleRGB=true
   ) {module=ScaleColor001}
   &DynamicMaterialParameters(
    Index 0 Param 1=&MakeFloatFromVector(
     Vector=&RotateVectorByQuaternion(Vector To Rotate=$Emitter.CapsuleGlowDirection, Quaternion=$Output.SocketLocation.SampledRotation)
     Channel=::NewEnumerator0
    )
    Index 0 Param 2=&MakeFloatFromVector(
     Vector=&RotateVectorByQuaternion(Vector To Rotate=$Emitter.CapsuleGlowDirection, Quaternion=$Output.SocketLocation.SampledRotation)
     Channel=::NewEnumerator1
    )
    Index 0 Param 3=&MakeFloatFromVector(
     Vector=&RotateVectorByQuaternion(Vector To Rotate=$Emitter.CapsuleGlowDirection, Quaternion=$Output.SocketLocation.SampledRotation)
     Channel=::NewEnumerator2
    )
    Index 0 Param 4=35
    Module.Param0WriteEnabled=true
    Module.Param1WriteEnabled=true
    Module.Param2WriteEnabled=true
    Module.Param3WriteEnabled=true
   ) {module=DynamicMaterialParameters001}
   &FN_Light_Attributes(
    Color=<rgba 0 0.576 1 1>
    Create Color Attribute=true
    Create Exponent Attribute=true
    Create Position Attribute=true
    Create Radius Attribute=true
    Exponent=1
    Position:NiagaraPosition=$Output.SocketLocation001.SampledPosition
    Radius=10
    Scale Color=&FloatFromCurve(
     FloatCurve=DataInterface() {propertyValues {Curve {keys [<at 0 0.5 RCIM_Cubic>,<at 0.5 1 RCIM_Cubic>,<at 1 0.5 RCIM_Cubic>]}}}
     Scale Curve=&Lerp_Float(A=0, B=10, Alpha=$System.ReactivityIntensity)
    )
   )
   &FN_Light_Attributes(
    Color=$System.Glow Color
    Create Color Attribute=true
    Create Exponent Attribute=true
    Create Position Attribute=true
    Create Radius Attribute=true
    Exponent=2
    Position=&AddVectorToPosition(Position=$Particles.Position, Vector=<0 -2.5 0>)
    Radius=20
    Scale Color=&FloatFromCurve(
     FloatCurve=DataInterface() {propertyValues {Curve {keys [<at 0 0.5 RCIM_Cubic>,<at 0.5 1 RCIM_Cubic>,<at 1 0.5 RCIM_Cubic>]}}}
     Scale Curve=&Lerp_Float(A=0, B=10, Alpha=$System.ReactivityIntensity)
    )
   ) {module=FN_Light_Attributes001}
  ]
  rendererClasses [&NiagaraSpriteRendererProperties, &NiagaraLightRendererProperties]
  renderers [
   {rendererClass=&NiagaraSpriteRendererProperties}
   {rendererClass=&NiagaraLightRendererProperties, rendererIndex=1}
   {rendererClass=&NiagaraLightRendererProperties, rendererIndex=2}
  ]
 }
 Glows {
  emitterUpdateScript [
   &EmitterState(
    Inactive Response=::NewEnumerator1
    Life Cycle Mode=::NewEnumerator1
    Loop Behavior=::NewEnumerator1
    Loop Duration Mode=::NewEnumerator1
    UseLoopDelay=false
   )
   &SpawnBurst_Instantaneous(Spawn Count=3, Use Spawn Probability=false, UseLoopCountLimit=false)
  ]
  particleSpawnScript [
   &PartitionParticles(Number Of Partitions=3)
   &InitializeParticle(
    Color=$System.Glow Color
    Lifetime=&MergeHelper_FloatSwitch(0=1.3, 1=0.8, 2=1, 3=?, Index=$Particles.PartitionParticles.Partition, Num=::NewEnumerator1)
    Position Offset=&MergeHelper_VectorSwitch(0=<-0.5 -8 13.7>, 1=<0 -8 13>, 2=<0 -8 13>, 3=?, Index=$Particles.PartitionParticles.Partition, Num=::NewEnumerator1)
    Sprite Rotation Mode=::NewEnumerator0
    Sprite Size=&MergeHelper_Vector2Switch(0=<30.5 30.5>, 1=<6 20>, 2=<6 38>, 3=?, 4=?, Index=$Particles.PartitionParticles.Partition, Num=::NewEnumerator1)
    Sprite Size Mode=::NewEnumerator4
    UsePositionOffset=true
    Write Material Random=true
   )
   &SpriteFacingAndAlignment(
    Sprite Alignment=&MergeHelper_VectorSwitch(0=<0 0 1>, 1=<0 0 1>, 2=<1 0 0>, 3=?, Index=$Particles.PartitionParticles.Partition, Num=::NewEnumerator1)
    Sprite Facing=&MergeHelper_VectorSwitch(0=<0 -1 0>, 1=<1 0 0>, 2=<0 0 1>, 3=?, Index=$Particles.PartitionParticles.Partition, Num=::NewEnumerator1)
    WriteAlignment=true
    WriteFacing=true
   )
   &CameraOffset(Camera Offset Amount=&MergeHelper_FloatSwitch(0=2, 1=20, 2=10, 3=?, Index=$Particles.PartitionParticles.Partition, Num=::NewEnumerator1))
   /SetVariables(Particles.VisibilityTag=$Particles.PartitionParticles.Partition) {isSetParameters=true}
  ]
  particleUpdateScript [
   &ParticleState(Kill Particles When Lifetime Has Elapsed=false, Loop Particles Lifetime=true)
   &ScaleColor(
    Scale Alpha=&FloatFromCurve(
     FloatCurve=DataInterface() {propertyValues {Curve {keys [
         <at 0 0.5 RCIM_Cubic>
         {arriveTangent=0.3214, interpMode=RCIM_Cubic, leaveTangent=0.3214, time=0.3, value=1}
         {arriveTangent=0.949, interpMode=RCIM_Cubic, leaveTangent=0.949, tangentMode=RCTM_User, time=0.7, value=0.725}
         <at 1 0.5 RCIM_Cubic>
        ]}}}
     Scale Curve=&Lerp_Float(A=0, B=&FN_FrontEndFloatModifier(Gameplay=0.2, Front End=0.1), Alpha=$System.ReactivityIntensity)
    )
    Scale RGB=&VectorFromFloat(Value=&Lerp_Float(
      A=0
      B=&FN_FrontEndFloatModifier(
       Gameplay=&Multiply_Float(
        A=&MergeHelper_FloatSwitch(0=1, 1=1, 2=1, 3=?, Index=$Particles.PartitionParticles.Partition, Num=::NewEnumerator1)
        B=&FNTimeOfDayScalar(Night Scale=0.6, Morning Scale=0.8, Evening Scale=1)
       )
       Front End=0.5
      )
      Alpha=$System.ReactivityIntensity
     ))
    ScaleA=true
    ScaleRGB=true
   )
  ]
  rendererClasses [&NiagaraSpriteRendererProperties]
  renderers [
   {rendererClass=&NiagaraSpriteRendererProperties}
   {rendererClass=&NiagaraSpriteRendererProperties, rendererIndex=1}
   {rendererClass=&NiagaraSpriteRendererProperties, rendererIndex=2}
  ]
 }
 Hamster_Speedlines {
  emitterUpdateScript [
   &EmitterState(Life Cycle Mode=::NewEnumerator1, Loop Behavior=::NewEnumerator1, Loop Duration Mode=::NewEnumerator1, UseLoopDelay=false)
   &SpawnRate(SpawnRate=&Lerp_Float(A=0, B=15, Alpha=$System.ReactivityIntensity), Use Spawn Probability=false)
  ]
  particleSpawnScript [
   &PartitionParticles()
   &InitializeParticle(
    Color=<rgba 0.5938 0.2184 0.0172 1>
    Lifetime Max=0.2
    Lifetime Min=0.1
    Lifetime Mode=::NewEnumerator1
    Mesh Scale=&Multiply_VectorByFloat(Vector=<0.95 1 1.1>, Float=0.13)
    Mesh Scale Mode=::NewEnumerator4
    Position=<0 -6 -5.8>
    Position Mode=::NewEnumerator1
    UsePositionOffset=false
    Write Material Random=true
   )
   &InitialMeshOrientation(
    Mesh Axis To Orient=<0 1 0>
    Mesh Orientation Mode=::NewEnumerator2
    Orientation Vector=<0 1 0>
    Rotation=&MakeVector(X=0, Y=&RandomRangeFloat(Minimum=0.05, Maximum=0.1), Z=0)
    Use Rotation Vector=true
   )
   &CameraOffset(Camera Offset Amount=&FN_FrontEndFloatModifier(Gameplay=10, Front End=10))
   &ShapeLocation(MaskSpawnGroup=false, Override Seed=false, Sphere Radius=0.5, UseOwnerScale=true)
  ]
  particleUpdateScript [
   &ParticleState()
   &UpdateMeshOrientation(Rotation Rate=&RandomRangeFloat(Minimum=-0.5, Maximum=-1.5), Rotation Vector=<0 1 0>)
   &ScaleColor(
    Scale Alpha=&FloatFromCurve(
     FloatCurve=DataInterface() {propertyValues {Curve {keys [
         {arriveTangent=12.6174, interpMode=RCIM_Cubic, leaveTangent=12.6174, tangentMode=RCTM_User, time=0, value=0}
         {interpMode=RCIM_Cubic, tangentMode=RCTM_User, time=0.2, value=1}
         {tangentMode=RCTM_User, time=1, value=0}
        ]}}}
     Scale Curve=&Lerp_Float(A=0, B=&FN_FrontEndFloatModifier(Gameplay=1, Front End=1), Alpha=$System.ReactivityIntensity)
    )
    Scale RGB=&MakeVectorFromLinearColorRGB(Color=&FNTimeOfDayColor(Night Color=<rgba 0.72 0.8882 1.5 1>, Morning Color=<rgba 1 1 1 1>, Day Color=<rgba 1 1 1 1>, Evening Color=<rgba 1 1 1 1>))
    ScaleA=true
    ScaleRGB=true
   )
   &OffsetPosition(Jitter Amount=0.5, Jitter Mode=1, Offset Mode=1)
  ]
  rendererClasses [&NiagaraMeshRendererProperties]
  renderers [{rendererClass=&NiagaraMeshRendererProperties, rendererIndex=0}]
 }
 Lightning_Splines {
  emitterUpdateScript [
   &EmitterState(
    Inactive Response=::NewEnumerator1
    Life Cycle Mode=::NewEnumerator1
    Loop Behavior=::NewEnumerator1
    Loop Duration Mode=::NewEnumerator1
    UseLoopDelay=false
   )
   &SpawnRate(
    SpawnRate=&FloatFromCurve(
     FloatCurve=DataInterface() {propertyValues {Curve {
        defaultValue=1
        keys [
         {interpMode=RCIM_Cubic, tangentMode=RCTM_User, time=0, value=0}
         {arriveTangent=2.1137, interpMode=RCIM_Cubic, leaveTangent=2.1137, tangentMode=RCTM_User, time=1, value=1}
        ]
       }}}
     CurveIndex=$System.ReactivityIntensity
     Scale Curve=20
    )
    Use Spawn Probability=false
   )
  ]
  particleSpawnScript [
   /SetVariables(
    Particles.InvertLightning=true
    Particles.LightningIndex:NiagaraInt32=&RandomRangeInt(Minimum=0, Maximum=6)
    Particles.MirrorLightning:NiagaraBool=&RandomBool()
   ) {isSetParameters=true}
   /SetVariables(Particles.MeshIndex=&SelectIntFromArray(Direct Array Index=$Particles.LightningIndex, Int Selection Array=DataInterface() {propertyValues {IntData [0,1,2,3,0,4,4]}})) {isSetParameters=true}
   &InitializeParticle(
    Lifetime=&RandomRangeFloat(Minimum=0.4, Maximum=0.5)
    Mesh Scale Mode=::NewEnumerator3
    Position Offset=&SelectVectorFromArray(
     Direct Array Index=$Particles.LightningIndex
     Vector Selection Array=DataInterface() {propertyValues {FloatData [<10 -8 -20>, <25 -10 -23>, <12 -12 -24>, <15 -10 -8>, <5 -8 -12>, <-2 -22 -4>, <2 -22 -10>]}}
    )
    UsePositionOffset=true
    Write Material Random=true
   )
   /SetVariables(Particles.Scale=&Multiply_VectorByFloat(
     Vector=&MergeHelper_VectorSwitch(
      0=&MakeCustomVectorFromBool(Bool=$Particles.MirrorLightning, True Vector=<-1 1 1>, False Vector=<1 1 1>)
      1=<-1 1 1>
      2=<1 1 1>
      3=<1 1.5 1>
      4=<1 1 1>
      5=<-1 1 1>
      6=<1 1 -1>
      Index=$Particles.LightningIndex
      Num=::NewEnumerator5
     )
     Float=&SelectFloatFromArray(
      Direct Array Index=$Particles.LightningIndex
      Float Selection Array=DataInterface() {propertyValues {FloatData [0.3,0.3,0.3,0.2,0.3,1.2,1.2]}}
     )
    )) {isSetParameters=true}
   &InitialMeshOrientation(
    Mesh Orientation Mode=::NewEnumerator2
    Orientation Vector=&SelectVectorFromArray(
     Direct Array Index=$Particles.LightningIndex
     Vector Selection Array=DataInterface() {propertyValues {FloatData [<1 0 0.3>, <1 0 0.5>, <1 -0.8 0>, <1 -0.8 0.1>, <1 0.3 0>, <0 1 0.25>, <0 1 0>]}}
    )
    Rotation=<0 0 0.25>
    Use Rotation Vector=true
   )
  ]
  particleUpdateScript [
   &ParticleState()
   &DynamicMaterialParameters(
    Index 0 Param 1=&RandomRangeFloat(Minimum=7, Maximum=8)
    Index 0 Param 2=&Divide_Float(
     A=&RandomRangeFloat(Minimum=1, Maximum=1.5)
     B=&SelectFloatFromArray(Direct Array Index=$Particles.LightningIndex, Float Selection Array=DataInterface() {propertyValues {FloatData [0.6,0.6,0.8,0.5,0.6]}})
    )
    Index 0 Param 3=&FloatFromCurve(FloatCurve=DataInterface() {propertyValues {Curve {keys [{arriveTangent=-9.7204, time=0, value=1}, <at 0.1 0.7>, <at 0.3 0.4>, <at 0.7 0.1>, <at 1 0>]}}})
    Index 0 Param 4=&MakeCustomFloatFromBool(Bool=$Particles.InvertLightning, True Float=1, False Float=0)
    Index 1 Param 1=1
    Index 1 Param 2=1
    Index 1 Param 3=1
    Module.Index0Param0Write=true
    Module.Index0Param1Write=true
    Module.Index0Param2Write=true
    Module.Index0Param3Write=true
    Module.Index1Param0Write=true
    Module.Index1Param1Write=true
    Module.Index1Param2Write=true
    Module.Index1Param3Write=true
    Write Parameter Index 1=true
   )
   &Color(Color=&ColorFromCurve(ColorCurve=DataInterface() {propertyValues {
       AlphaCurve {keys [<at 0 0>, {arriveTangent=0.3873, interpMode=RCIM_Cubic, leaveTangent=0.3873, time=0.0998, value=1}, <at 0.7746 0.3>, <at 1 0 RCIM_Cubic>]}
       BlueCurve {keys [<at 0 10 RCIM_Cubic>,<at 0.4311 5>,<at 0.8167 1 RCIM_Cubic>]}
       GreenCurve {keys [<at 0 7.1305 RCIM_Cubic>,<at 0.4311 2>,<at 0.8167 0.3333 RCIM_Cubic>]}
       RedCurve {keys [<at 0 0 RCIM_Cubic>,<at 0.4311 0>,<at 0.8167 0 RCIM_Cubic>]}
      }}))
  ]
  rendererClasses [&NiagaraMeshRendererProperties]
  renderers [{rendererClass=&NiagaraMeshRendererProperties, rendererIndex=0}]
 }
 Motion_Blur {
  emitterUpdateScript [
   &EmitterState(
    Inactive Response=::NewEnumerator1
    Life Cycle Mode=::NewEnumerator1
    Loop Behavior=::NewEnumerator1
    Loop Duration Mode=::NewEnumerator1
    UseLoopDelay=false
   )
   &SpawnBurst_Instantaneous(Use Spawn Probability=false, UseLoopCountLimit=false)
  ]
  particleSpawnScript [
   &InitializeParticle(
    Color=<rgba 0.1479 0.4772 0.6927 0.2>
    Position Offset=<0 -12 -5>
    Sprite Size Mode=::NewEnumerator3
    Uniform Sprite Size=20
    UsePositionOffset=true
    Write Material Random=true
   )
   &CameraOffset(Camera Offset Amount=1)
  ]
  particleUpdateScript [
   &ParticleState(Kill Particles When Lifetime Has Elapsed=false, Let Infinitely Lived Particles Die When Emitter Deactivates=true)
   &ScaleColor(
    Scale Alpha=&Lerp_Float(A=0, B=1, Alpha=$System.ReactivityIntensity)
    Scale RGB=&VectorFromFloat(Value=&FNTimeOfDayScalar(Night Scale=1, Morning Scale=1, Evening Scale=1))
    ScaleA=true
    ScaleRGB=true
   )
   &SpriteFacingAndAlignment(Sprite Facing=<0 -1 0>, WriteAlignment=false, WriteFacing=true)
   &DynamicMaterialParameters(
    Index 0 Param 1=0.2
    Module.Param0WriteEnabled=true
    Module.Param1WriteEnabled=false
    Module.Param2WriteEnabled=false
    Module.Param3WriteEnabled=false
   )
  ]
  rendererClasses [&NiagaraSpriteRendererProperties]
  renderers [{rendererClass=&NiagaraSpriteRendererProperties}]
 }
 Sparks {
  emitterUpdateScript [
   &InterpolateOverTime(
    Rate Of Change=3
    Target Float Value:NiagaraFloat=&MakeCustomFloatFromBool(
     Bool=&SetBoolByFloatComparison(A=&VectorLength(Vector=$Engine.Owner.Velocity), B=100)
     True Float=0
     False Float=1
    )
   )
   &EmitterState()
   &SpawnRate(
    SpawnRate=&FloatFromCurve(
     FloatCurve=DataInterface() {propertyValues {Curve {
        defaultValue=1
        keys [
         {interpMode=RCIM_Cubic, tangentMode=RCTM_User, time=0, value=0}
         {arriveTangent=1.5797, interpMode=RCIM_Cubic, leaveTangent=9.2677, tangentMode=RCTM_Break, time=0.9, value=1}
         {arriveTangent=1, interpMode=RCIM_Constant, time=1, value=3}
        ]
       }}}
     CurveIndex=$System.ReactivityIntensity
     Scale Curve=16.6667
    )
    Use Spawn Probability=true
   )
  ]
  particleSpawnScript [
   &InitializeParticle(
    Color=<rgba 0.2 0.7253 1 1>
    Lifetime Max=0.6
    Lifetime Min=0.3
    Lifetime Mode=::NewEnumerator1
    Sprite Size Max=<1.5 4>
    Sprite Size Min=<1 3>
    Sprite Size Mode=::NewEnumerator2
    UsePositionOffset=false
    Write Material Random=true
   )
   &SocketLocation(
    Interpolate Socket Sampling=true
    SamplePosOffset=true
    Sampled Position Offset=<0 -8 -11>
    Sampled Position Offset Coordinate Space:ENiagaraSkelMeshTransforms=::NewEnumerator2
    Socket Index=1
    Socket Reader=$System.Socket Reader
    Socket Sampling Mode=::NewEnumerator3
    VelocitySpeedLimit=false
   )
   &ShapeLocation(
    Cylinder Distribution=::NewEnumerator2
    Cylinder Height=10
    Cylinder Height Midpoint=0
    Cylinder Radius=1
    MaskSpawnGroup=false
    Rotation Axis=<0 1 0>
    Rotation Mode=::NewEnumerator1
    Shape Primitive=::NewEnumerator1
    U Position=&RandomRangeFloat(Minimum=0, Maximum=1)
    UseOwnerScale=true
    V Position=&RandomRangeFloat(Minimum=0, Maximum=1)
   ) {module=ShapeLocation001}
   &AddVelocity(
    Cone Angle=30
    Cone Axis=<1 0 0.2>
    Module.UseConeFalloff=false
    Module.UseVelDistribution=false
    Override Seed=false
    Velocity Mode=::NewEnumerator2
    Velocity Speed=&Multiply_Float(
     A=&RandomRangeFloat(Minimum=100, Maximum=250)
     B=&FloatFromCurve(
      FloatCurve=DataInterface() {propertyValues {Curve {
         defaultValue=0
         keys [
          {arriveTangent=2.962, interpMode=RCIM_Cubic, leaveTangent=2.962, tangentMode=RCTM_User, time=0, value=0}
          {interpMode=RCIM_Cubic, tangentMode=RCTM_User, time=1, value=1}
         ]
        }}}
      CurveIndex=$System.ReactivityIntensity
     )
    )
   )
   &CameraOffset()
  ]
  particleUpdateScript [
   &ParticleState()
   &SocketLocation(
    Position Sampling=::NewEnumerator1
    SamplePosOffset=false
    Socket Reader=$System.Socket Reader
    Socket Sampling Mode=::NewEnumerator3
    VelocitySpeedLimit=false
   ) {module=SocketLocation001}
   /SetVariables(Particles.Glow03_Color=&MakeLinearColorFromVectorAndFloat(
     "Vector (RGB)"=&MakeVectorFromLinearColorRGB(Color=&MultiplyLinearColor(Linear Color=<rgba 0 0.3493 1 1>, Scale Factor=&MakeVector4_FromFloat(Float=1)))
     "Float (Alpha)"=0.2
    )) {isSetParameters=true}
   &ScaleColor(
    Scale Alpha=&FloatFromCurve(FloatCurve=DataInterface() {propertyValues {Curve {
        defaultValue=1
        keys [
         {interpMode=RCIM_Cubic, tangentMode=RCTM_User, time=0, value=0}
         {interpMode=RCIM_Cubic, tangentMode=RCTM_User, time=0.5, value=1}
         {tangentMode=RCTM_User, time=1, value=0}
        ]
       }}})
    ScaleA=true
    ScaleRGB=true
   )
   &ScaleSpriteSize(Scale Factor=&Vector2DFromFloat(Value=&FloatFromCurve(FloatCurve=DataInterface() {propertyValues {Curve {
         defaultValue=0
         keys [
          {arriveTangent=9.0725, interpMode=RCIM_Cubic, leaveTangent=9.0725, tangentMode=RCTM_User, time=0, value=0}
          {arriveTangent=-2.5249, interpMode=RCIM_Cubic, leaveTangent=-2.5249, tangentMode=RCTM_User, time=0.1408, value=0.9847}
          {arriveTangent=-0.0405, interpMode=RCIM_Cubic, leaveTangent=-0.0405, tangentMode=RCTM_User, time=1, value=0}
         ]
        }}}))) {module=ScaleSpriteSize001}
   &CurlNoiseForce(Noise Frequency=300, Noise Strength=300, Pan Noise=true, Pan Noise Field=<0 0 1>)
   &Drag(Drag=&RandomRangeFloat(Minimum=7, Maximum=10), Use Linear Drag=true, Use Rotational Drag=false)
   &GravityForce(Coordinate Space=::Local, Gravity=<0 0 30>)
   &InheritSourceMovement(
    Applied Position Scale=&VectorFromFloat(Value=&FloatFromCurve(FloatCurve=DataInterface() {propertyValues {Curve {
         defaultValue=1
         keys [
          {interpMode=RCIM_Cubic, tangentMode=RCTM_User, time=0, value=1}
          {arriveTangent=-2.0317, interpMode=RCIM_Cubic, leaveTangent=-2.0317, tangentMode=RCTM_User, time=1, value=0}
         ]
        }}}))
    ApplyForce=false
    ApplyPosition=true
    ApplyPositionFalloff=false
    Source Velocity:Vector3f=$Output.SocketLocation001.SampledVelocity
   )
   &SolveForcesAndVelocity(Clamp Velocity=false, Limit Acceleration=false)
   &ScaleSpriteSizeBySpeed(Max Scale Factor=<0.5 2>, Velocity Threshold=200)
   /SetVariables(Particles.Glow03_Size=&Multiply_Vector2D_ByFloat(Float=3, Vector2D=$Particles.SpriteSize)) {isSetParameters=true}
  ]
  rendererClasses [&NiagaraSpriteRendererProperties]
  renderers [
   {rendererClass=&NiagaraSpriteRendererProperties}
   {rendererClass=&NiagaraSpriteRendererProperties, rendererIndex=1}
  ]
 }
 Wheel_Speedlines {
  emitterUpdateScript [
   &EmitterState(
    Inactive Response=::NewEnumerator1
    Life Cycle Mode=::NewEnumerator1
    Loop Behavior=::NewEnumerator1
    Loop Duration Mode=::NewEnumerator1
    UseLoopDelay=false
   )
   &SpawnBurst_Instantaneous(Use Spawn Probability=false, UseLoopCountLimit=false)
  ]
  particleSpawnScript [
   &InitializeParticle(
    Color=<rgba 0.0839 0.4485 1 0.33>
    Lifetime=5
    Mesh Scale Mode=::NewEnumerator3
    Mesh Uniform Scale=0.5
    UsePositionOffset=false
    Write Material Random=true
   )
   &InitialMeshOrientation(Mesh Orientation Mode=::NewEnumerator2, Use Rotation Vector=false)
   &CameraOffset(Camera Offset Amount=&FN_FrontEndFloatModifier(Gameplay=2, Front End=1))
  ]
  particleUpdateScript [
   &ParticleState(Kill Particles When Lifetime Has Elapsed=false, Let Infinitely Lived Particles Die When Emitter Deactivates=true)
   &SocketLocation(
    SamplePosOffset=true
    Sampled Position Offset=<0 -3 2>
    Socket Index=1
    Socket Reader=$System.Socket Reader
    Socket Sampling Mode=::NewEnumerator3
    VelocitySpeedLimit=false
   )
   &UpdateMeshOrientation(Rotation Vector=<0 1 0>)
   &ScaleColor(
    Scale Alpha=&FloatFromCurve(
     FloatCurve=DataInterface() {propertyValues {Curve {
        defaultValue=0
        keys [{interpMode=RCIM_Cubic, tangentMode=RCTM_User, time=0, value=0}, {interpMode=RCIM_Cubic, tangentMode=RCTM_User, time=1, value=1}]
       }}}
     CurveIndex=$System.ReactivityIntensity
     Scale Curve=10
    )
    ScaleA=true
    ScaleRGB=true
   )
  ]
  rendererClasses [&NiagaraMeshRendererProperties]
  renderers [{rendererClass=&NiagaraMeshRendererProperties, rendererIndex=0}]
 }
}
summary {
 systemName=NS_Backpack_EssayViewMyth_Idle
 emitters [
  {
   emitterName=Capsule_Glow_And_Light
   rendererClasses [&NiagaraSpriteRendererProperties, &NiagaraLightRendererProperties]
  }
  {emitterName=Lightning_Splines, rendererClasses [&NiagaraMeshRendererProperties]}
  {emitterName=Sparks, rendererClasses [&NiagaraSpriteRendererProperties]}
  {emitterName=Wheel_Speedlines, rendererClasses [&NiagaraMeshRendererProperties]}
  {emitterName=Motion_Blur, rendererClasses [&NiagaraSpriteRendererProperties]}
  {emitterName=Hamster_Speedlines, rendererClasses [&NiagaraMeshRendererProperties]}
  {emitterName=Glows, rendererClasses [&NiagaraSpriteRendererProperties]}
 ]
 userVariables [
  User.bIsFrontEnd {type=NiagaraBool}
  User.bIsFrontEndPreview {type=NiagaraBool}
  User.DebugReactivity {type=NiagaraBool}
  User.ReactivityTime {type=NiagaraFloat, defaultValue=0, description=""}
 ]
}
