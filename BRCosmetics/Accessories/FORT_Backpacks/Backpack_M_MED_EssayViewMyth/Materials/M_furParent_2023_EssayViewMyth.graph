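# Parent material for the EssayViewMyth backpack, layering optional fur and reactive emissive glow between the shared 2023 character PreFX and PostFX functions.
# A time-since-trigger curve atlas lookup drives reactivity, with a scalar override for the frontend.
# Mask channels separate low-to-high energy color, red sign and blue Fresnel emissive terms.
# The fur result blends in by hair mask, then base color is selectively darkened and emissive is added.
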Material /BRCosmetics/Accessories/FORT_Backpacks/Backpack_M_MED_EssayViewMyth/Materials/M_furParent_2023_EssayViewMyth
graph {
 properties {
  bUseMaterialAttributes=true
  bUsedWithInstancedStaticMeshes=true
  bUsedWithMeshDeformer=true
  bUsedWithMorphTargets=true
  bUsedWithSkeletalMesh=true
  pixelDepthOffsetMode=PDOM_Legacy
  shadingModel=MSM_FromMaterialExpression
  usedShadingModels=Default Lit | Subsurface | Clear Coat | Subsurface Profile | Cloth
 }
 nodes [
  n1 = MaterialFunctionCall() {materialFunction=&MF_CharacterParent_2023_PreFX}
  n2 = MaterialFunctionCall!(In=@n1.Result, Mask Input=@n1.Result) {materialFunction=/Game/Global/Materials/Character_Parent_2023/Fur_Parent_2023/MF_Fur_Deprecated}
  n3 = GetMaterialAttributes(None=BlendMaterialAttributes(
    A=@n1.Result
    B=StaticSwitchParameter(True=@n2.Result, False=@n1.Result) {defaultValue=true, group=Skin FX, parameterName=UseFNHair2_0}
    Alpha=@n2.HairMask
   )) {attributeGetTypes [#1, #2, #3]}
  n4 = NamedRerouteUsage()
  n5 = LinearInterpolate(A=@n3.Base Color, Alpha=NamedRerouteUsage()) {constB=0}
  n6 = Power(Base=NamedRerouteUsage())
  MaterialFunctionCall(
   Material=SetMaterialAttributes!(
    MaterialAttributes=@n3.MaterialAttributes
    Emissive Color=Add(A=@n3.Emissive Color, B=@n4)
    Base Color=LinearInterpolate(A=@n3.Base Color, B=@n5, Alpha=@n6)
   ) {attributeSetTypes [#1, #2]}
   M=@n1.M
   Pre Skinned Local Position=@n1.M
   Distance Blend=@n1.M
   Fresnel=@n1.M
  ) {materialFunction=&MF_CharacterParent_2023_PostFX}
  n7 = NamedRerouteUsage()
  MaterialFunctionCall(Normal=@n7, Flatness=Constant() {r=-1}) {materialFunction=&FlattenNormal}
  n8 = TextureSampleParameter2D() {
   group=Reactivity
   parameterName=TechArtMask_RE
   samplerType=SAMPLERTYPE_Masks
   texture=/BRCosmetics/Accessories/FORT_Backpacks/Backpack_M_MED_EssayViewMyth/Textures/T_EssayViewMyth_Backpack_FX
  }
  n9 = ScalarParameter() {defaultValue=1, group=Colorshading, parameterName=HighEnergyIntensity_Low}
  n10 = ScalarParameter() {defaultValue=2, group=Colorshading, parameterName=HighEnergyIntensity}
  n11 = NamedRerouteUsage()
  n12 = VectorParameter() {group=Colorshading, parameterName=LowEnergyColor, defaultValue=<rgba 0.2615 0.0547 0.0547 1>}
  n13 = Multiply(A=@n12.RGB, B=@n12.RGB)
  n14 = VectorParameter() {group=Colorshading, parameterName=HighEnergyColor, defaultValue=<rgba 1 1 1 1>}
  n15 = Multiply(A=@n14.RGB, B=@n14.RGB)
  n16 = Subtract(A=Time(), B=ScalarParameter() {group=Reactivity, parameterName=ReactivityTime})
  n17 = ScalarParameter() {defaultValue=4, parameterName=ReactivityDuration}
  n18 = CurveAtlasRowParameter(CurveTime=Saturate(None=Divide(A=@n16, B=@n17))) {
   atlas=&/T_AnimationPhase
   curve=&/C_AnimationPhase
   group=Reactivity
   parameterName=AnimationControl
  }
  n19 = Saturate(None=LinearInterpolate(
    A=@n18.R
    B=ScalarParameter() {defaultValue=1, group=Reactivity, parameterName=ReactiveAlpha}
    Alpha=ScalarParameter() {group=Reactivity, parameterName=IsInFrontend}
   ))
  n20 = Power(Base=@n19) {constExponent=3}
  n21 = NamedRerouteDeclaration(None=@n19) {name=ReactivityAlpha, nodeColor=<rgba 1 0 0.0703 1>}
  n22 = Multiply(
   A=Multiply(A=@n8.G, B=LinearInterpolate(A=@n9, B=@n10, Alpha=@n11))
   B=Multiply(A=LinearInterpolate(A=@n13, B=@n15, Alpha=@n20), B=Power(Base=@n21) {constExponent=0.4})
  )
  n23 = VectorParameter() {group=Colorshading, parameterName=RedEM, defaultValue=<rgba 1 0 0 1>}
  n24 = Multiply(A=Multiply(A=@n23.RGB, B=@n23.RGB), B=NamedRerouteUsage())
  n25 = VectorParameter() {group=Colorshading, parameterName=BlueEmissive, defaultValue=<rgba 0 0.2797 1.5 1>}
  n26 = Transform(None=@n7)
  n27 = Multiply(A=Multiply(A=Multiply(A=@n25.RGB, B=@n25.RGB), B=Fresnel(Normal=@n26) {baseReflectFraction=0, exponent=1}), B=NamedRerouteUsage())
  n28 = NamedRerouteUsage()
  NamedRerouteDeclaration(None=Add(A=LinearInterpolate(A=@n22, B=@n24, Alpha=@n8.R), B=Multiply(A=@n27, B=@n28))) {name=EmissiveOUT, nodeColor=<rgba 0 1 0.4141 1>}
  NamedRerouteDeclaration(None=Max(A=@n8.R, B=@n8.R)) {name=SignArea, nodeColor=<rgba 0 0.6484 1 1>}
  NamedRerouteDeclaration(None=@n8.B) {name=BlueArea, nodeColor=<rgba 0.0781 0 1 1>}
  NamedRerouteDeclaration(None=@n3.Normal) {name=Normal, nodeColor=<rgba 0 0.0859 1 1>}
  Reroute(None=@n1.Tech Art Mask)
 ]
}
