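# Animation Blueprint for the EssayViewMyth backpack, playing idle and elimination-reaction animations.
# bIsInFrontEnd selects between a lobby machine with timed reactions, gated by the ChiveFlake modifier, and a machine driven by the enemy-elimination input event.
# Additive bone edits then spin the wheel by integrating the WheelSpinRate curve and shift the pack from spine yaw.
# Three TESTING state machines are not connected to the output.
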AnimBlueprint /BRCosmetics/Accessories/FORT_Backpacks/Backpack_M_MED_EssayViewMyth/Meshes/M_MED_EssayViewMyth_Pack_AnimBP
target_skeleton=&/M_MED_EssayViewMyth_Pack_Skeleton
graphs [
 AnimGraph {nodes [
   Input Pose()
   AnimGraphNode_StateMachine_0 = FE Idle + Reactive() {is=title AnimGraphNode_StateMachine}
   AnimGraphNode_StateMachine_1 = Idle and Elimination() {is=title AnimGraphNode_StateMachine}
   n1 = Get bIsInFrontEnd()
   n2 = "Transform (Modify) Bone - Bone: jnt_Wheel"(
    ComponentPose=Local To Component(LocalPose=Blend Poses by bool(BlendPose_0=@AnimGraphNode_StateMachine_0, BlendPose_1=@AnimGraphNode_StateMachine_1, bActiveValue=@n1) {blendType=Linear})
    Alpha=1
    Translation="0,0,0"
    Rotation=Get Wheel Rotation() -> (Wheel Rotation:Rotator)
    Scale="1,1,1"
   ) {rotationMode=BMM_Additive, translationMode=BMM_Additive, alphaBoolBlend {blendOption=Linear}, boneToModify {boneName=jnt_Wheel}}
   n3 = "Transform (Modify) Bone - Bone: pack"(ComponentPose=@n2, Translation="0,-1,0", Rotation="0,0,0") {
    translationMode=BMM_Additive
    translationSpace=BCS_BoneSpace
    alphaBoolBlend {blendOption=Linear}
    boneToModify {boneName=pack}
    translation {y=-1}
   }
   n4 = "Transform (Modify) Bone - Bone: pack"(ComponentPose=@n3, Translation=Get backpack_translation(), Rotation=Get backpack_rotate()) {
    rotationMode=BMM_Additive
    rotationSpace=BCS_BoneSpace
    translationMode=BMM_Additive
    translationSpace=BCS_BoneSpace
    alphaBoolBlend {blendOption=Linear}
    boneToModify {boneName=pack}
   }
   Output Pose(Result=Component To Local(ComponentPose=@n4)) {layerGroup=None}
   AnimGraphNode_StateMachine_2 = TESTING - ElimB() {is=title AnimGraphNode_StateMachine}
   AnimGraphNode_StateMachine_3 = TESTING - Capture() {is=title AnimGraphNode_StateMachine}
   AnimGraphNode_StateMachine_4 = TESTING - WheelSpin() {is=title AnimGraphNode_StateMachine}
  ]}
 FE Idle + Reactive {
  entry=FE_Idle
  path=AnimGraph.AnimGraphNode_StateMachine_0.FE Idle + Reactive
  ids=id
  nodes [
   AnimStateNode_0 = FE_Idle() {is=title AnimStateNode}
   AnimStateNode_1 = FE_Elim() {is=title AnimStateNode, stateEntered {guid=#1, notifyName=Entered FE_Reactive_Anim}}
   AnimStateNode_2 = FE_Elim_Out() {is=title AnimStateNode}
   AnimStateTransitionNode_0 = AnimStateTransitionNode() {bAutomaticRuleBasedOnSequencePlayerInState=true, crossfadeDuration=0.001}
   AnimStateTransitionNode_2 = AnimStateTransitionNode() {bAutomaticRuleBasedOnSequencePlayerInState=true, crossfadeDuration=0.001}
   AnimStateTransitionNode_3 = AnimStateTransitionNode() {bAutomaticRuleBasedOnSequencePlayerInState=true}
  ]
  links [
   {
    dst=FE_Elim_Out
    rule=AnimGraph.AnimGraphNode_StateMachine_0.FE Idle + Reactive.AnimStateTransitionNode_2.Transition
    src=FE_Elim
    via=AnimStateTransitionNode_2
   }
   {
    dst=FE_Idle
    rule=AnimGraph.AnimGraphNode_StateMachine_0.FE Idle + Reactive.AnimStateTransitionNode_3.Transition
    src=FE_Elim_Out
    via=AnimStateTransitionNode_3
   }
   {
    dst=FE_Elim
    rule=AnimGraph.AnimGraphNode_StateMachine_0.FE Idle + Reactive.AnimStateTransitionNode_0.Transition
    src=FE_Idle
    via=AnimStateTransitionNode_0
   }
  ]
 }
 FE_Idle {
  path=AnimGraph.AnimGraphNode_StateMachine_0.FE Idle + Reactive.AnimStateNode_0.FE_Idle
  nodes [
   n1 = Sequence Player 'EssayViewMyth_Idle_PACK'() {bLoopAnimation=false, sequence=/BRCosmetics/Accessories/FORT_Backpacks/Backpack_M_MED_EssayViewMyth/Animation/EssayViewMyth_Idle_PACK}
   Output Animation Pose(Result=@n1)
  ]
 }
 FE_Elim {
  path=AnimGraph.AnimGraphNode_StateMachine_0.FE Idle + Reactive.AnimStateNode_1.FE_Elim
  nodes [
   n1 = Sequence Player 'EssayViewMyth_ElimA_PACK'() {bLoopAnimation=false, sequence=/BRCosmetics/Accessories/FORT_Backpacks/Backpack_M_MED_EssayViewMyth/Animation/EssayViewMyth_ElimA_PACK}
   Output Animation Pose(Result=@n1) {
    node {stateEntryFunction {functionName=OnStateEntry}}
    stateEntryFunction {bSelfContext=true, memberGuid=#2, memberName=OnStateEntry}
   }
  ]
 }
 FE_Elim_Out {
  path=AnimGraph.AnimGraphNode_StateMachine_0.FE Idle + Reactive.AnimStateNode_2.FE_Elim_Out
  nodes [
   n1 = Random Sequence Player!() -> (Pose:Pose Link Structure) {
    is=title AnimGraphNode_RandomPlayer
    entries [
     {
      chanceToPlay=0.7
      maxLoopCount=0
      maxPlayRate=1
      minLoopCount=0
      minPlayRate=1
      sequence=/BRCosmetics/Accessories/FORT_Backpacks/Backpack_M_MED_EssayViewMyth/Animation/EssayViewMyth_Elim_Out_PACK
      blendIn {blendOption=Linear, blendTime=0.001, customCurve=null}
     }
     {
      chanceToPlay=0.3
      maxLoopCount=0
      maxPlayRate=1
      minLoopCount=0
      minPlayRate=1
      sequence=/BRCosmetics/Accessories/FORT_Backpacks/Backpack_M_MED_EssayViewMyth/Animation/EssayViewMyth_ElimB_PACK
      blendIn {blendOption=Linear, blendTime=0.001, customCurve=null}
     }
    ]
   }
   Output Animation Pose(Result=@n1)
  ]
 }
 Transition {
  path=AnimGraph.AnimGraphNode_StateMachine_0.FE Idle + Reactive.AnimStateTransitionNode_0.Transition
  nodes [Result(bCanEnterTransition=Get Play Reactive FE Anim() -> (Play Reactive FE Anim:bool))]
 }
 Transition {path=AnimGraph.AnimGraphNode_StateMachine_0.FE Idle + Reactive.AnimStateTransitionNode_2.Transition, nodes [Result()]}
 Transition {path=AnimGraph.AnimGraphNode_StateMachine_0.FE Idle + Reactive.AnimStateTransitionNode_3.Transition, nodes [Result()]}
 Idle and Elimination {
  entry=Idle
  path=AnimGraph.AnimGraphNode_StateMachine_1.Idle and Elimination
  ids=id
  nodes [
   AnimStateNode_0 = Idle() {
    stateEntered {guid=#3}
    stateFullyBlended {guid=#4}
    stateLeft {guid=#5}
   }
   AnimStateNode_1 = Elimination() {
    is=title AnimStateNode
    stateEntered {guid=#6}
    stateFullyBlended {guid=#7}
    stateLeft {guid=#8}
   }
   AnimStateNode_2 = Elimination_Out() {is=title AnimStateNode}
   AnimStateTransitionNode_0 = AnimStateTransitionNode() {
    crossfadeDuration=0.001
    transitionEnd {guid=#9}
    transitionInterrupt {guid=#10}
    transitionStart {guid=#11}
   }
   AnimStateTransitionNode_1 = AnimStateTransitionNode() {bAutomaticRuleBasedOnSequencePlayerInState=true, crossfadeDuration=0.001}
   AnimStateTransitionNode_2 = AnimStateTransitionNode() {bAutomaticRuleBasedOnSequencePlayerInState=true}
  ]
  links [
   {
    dst=Elimination_Out
    rule=AnimGraph.AnimGraphNode_StateMachine_1.Idle and Elimination.AnimStateTransitionNode_1.Transition
    src=Elimination
    via=AnimStateTransitionNode_1
   }
   {
    dst=Idle
    rule=AnimGraph.AnimGraphNode_StateMachine_1.Idle and Elimination.AnimStateTransitionNode_2.Transition
    src=Elimination_Out
    via=AnimStateTransitionNode_2
   }
   {
    dst=Elimination
    rule=AnimGraph.AnimGraphNode_StateMachine_1.Idle and Elimination.AnimStateTransitionNode_0.Transition
    src=Idle
    via=AnimStateTransitionNode_0
   }
  ]
 }
 Idle {
  path=AnimGraph.AnimGraphNode_StateMachine_1.Idle and Elimination.AnimStateNode_0.Idle
  nodes [
   Sequence Player 'ChiveFlake_Idle_PACK'() {sequence=/BRCosmetics/Accessories/FORT_Backpacks/Backpack_F_MED_ChiveFlake/Animations/ChiveFlake_Idle_PACK}
   Output Animation Pose(Result=Sequence Player 'EssayViewMyth_Idle_PACK'() {sequence=&EssayViewMyth_Idle_PACK})
  ]
 }
 Elimination {
  path=AnimGraph.AnimGraphNode_StateMachine_1.Idle and Elimination.AnimStateNode_1.Elimination
  nodes [
   n1 = Sequence Player 'EssayViewMyth_ElimA_PACK'() {bLoopAnimation=false, sequence=&EssayViewMyth_ElimA_PACK}
   Output Animation Pose(Result=@n1)
  ]
 }
 Elimination_Out {
  path=AnimGraph.AnimGraphNode_StateMachine_1.Idle and Elimination.AnimStateNode_2.Elimination_Out
  nodes [
   n1 = Random Sequence Player() {entries [
     {
      chanceToPlay=0.7
      maxLoopCount=0
      maxPlayRate=1
      minLoopCount=0
      minPlayRate=1
      sequence=&EssayViewMyth_Elim_Out_PACK
      blendIn {blendOption=Linear, blendTime=0.001, customCurve=null}
     }
     {
      chanceToPlay=0.3
      maxLoopCount=0
      maxPlayRate=1
      minLoopCount=0
      minPlayRate=1
      sequence=&EssayViewMyth_ElimB_PACK
      blendIn {blendOption=Linear, blendTime=0.001, customCurve=null}
     }
    ]}
   Output Animation Pose(Result=@n1)
  ]
 }
 Transition {
  path=AnimGraph.AnimGraphNode_StateMachine_1.Idle and Elimination.AnimStateTransitionNode_0.Transition
  nodes [Result(bCanEnterTransition=Get bEnemyPlayerEliminated() -> (bEnemyPlayerEliminated:bool))]
 }
 Transition {path=AnimGraph.AnimGraphNode_StateMachine_1.Idle and Elimination.AnimStateTransitionNode_1.Transition, nodes [Result()]}
 Transition {path=AnimGraph.AnimGraphNode_StateMachine_1.Idle and Elimination.AnimStateTransitionNode_2.Transition, nodes [Result()]}
 TESTING - ElimB {
  entry=ElimB
  path=AnimGraph.AnimGraphNode_StateMachine_2.TESTING - ElimB
  ids=id
  nodes [
   AnimStateNode_0 = Idle_04() {is=title AnimStateNode}
   AnimStateNode_5 = ElimB() {
    is=title AnimStateNode
    stateEntered {guid=#12}
    stateFullyBlended {guid=#13}
    stateLeft {guid=#14}
   }
   AnimStateNode_7 = IdleBreak() {is=title AnimStateNode}
  ]
 }
 Idle_04 {
  path=AnimGraph.AnimGraphNode_StateMachine_2.TESTING - ElimB.AnimStateNode_0.Idle_04
  nodes [
   n1 = Sequence Player 'EssayViewMyth_Idle_PACK'() {bLoopAnimation=false, sequence=&EssayViewMyth_Idle_PACK}
   Output Animation Pose(Result=@n1)
  ]
 }
 ElimB {
  path=AnimGraph.AnimGraphNode_StateMachine_2.TESTING - ElimB.AnimStateNode_5.ElimB
  nodes [Output Animation Pose(Result=Sequence Player 'EssayViewMyth_ElimB_PACK'() {sequence=&EssayViewMyth_ElimB_PACK})]
 }
 IdleBreak {path=AnimGraph.AnimGraphNode_StateMachine_2.TESTING - ElimB.AnimStateNode_7.IdleBreak, nodes [Output Animation Pose()]}
 TESTING - Capture {
  entry=Idle_01
  path=AnimGraph.AnimGraphNode_StateMachine_3.TESTING - Capture
  ids=id
  nodes [
   AnimStateNode_0 = Idle_04()
   AnimStateNode_1 = Idle_02() {is=title AnimStateNode}
   AnimStateNode_2 = ElimA_01() {is=title AnimStateNode}
   AnimStateNode_3 = Elim_Out() {is=title AnimStateNode}
   AnimStateNode_4 = ElimA_02() {is=title AnimStateNode}
   AnimStateNode_5 = ElimB()
   AnimStateNode_6 = Idle_01() {is=title AnimStateNode}
   AnimStateNode_7 = IdleBreak()
   AnimStateNode_8 = Idle_03() {is=title AnimStateNode}
   AnimStateTransitionNode_0 = AnimStateTransitionNode() {bAutomaticRuleBasedOnSequencePlayerInState=true, crossfadeDuration=0.001}
   AnimStateTransitionNode_1 = AnimStateTransitionNode() {bAutomaticRuleBasedOnSequencePlayerInState=true, crossfadeDuration=0.001}
   AnimStateTransitionNode_2 = AnimStateTransitionNode() {bAutomaticRuleBasedOnSequencePlayerInState=true, crossfadeDuration=0.001}
   AnimStateTransitionNode_3 = AnimStateTransitionNode() {bAutomaticRuleBasedOnSequencePlayerInState=true, crossfadeDuration=0.001}
   AnimStateTransitionNode_4 = AnimStateTransitionNode() {bAutomaticRuleBasedOnSequencePlayerInState=true, crossfadeDuration=0.001}
   AnimStateTransitionNode_5 = AnimStateTransitionNode() {bAutomaticRuleBasedOnSequencePlayerInState=true, crossfadeDuration=0.001}
   AnimStateTransitionNode_6 = AnimStateTransitionNode() {bAutomaticRuleBasedOnSequencePlayerInState=true, crossfadeDuration=0.001}
  ]
  links [
   {
    dst=Elim_Out
    rule=AnimGraph.AnimGraphNode_StateMachine_3.TESTING - Capture.AnimStateTransitionNode_2.Transition
    src=ElimA_01
    via=AnimStateTransitionNode_2
   }
   {
    dst=ElimB
    rule=AnimGraph.AnimGraphNode_StateMachine_3.TESTING - Capture.AnimStateTransitionNode_4.Transition
    src=ElimA_02
    via=AnimStateTransitionNode_4
   }
   {
    dst=Idle_01
    rule=AnimGraph.AnimGraphNode_StateMachine_3.TESTING - Capture.AnimStateTransitionNode_5.Transition
    src=ElimB
    via=AnimStateTransitionNode_5
   }
   {
    dst=Idle_03
    rule=AnimGraph.AnimGraphNode_StateMachine_3.TESTING - Capture.AnimStateTransitionNode_3.Transition
    src=Elim_Out
    via=AnimStateTransitionNode_3
   }
   {
    dst=Idle_02
    rule=AnimGraph.AnimGraphNode_StateMachine_3.TESTING - Capture.AnimStateTransitionNode_0.Transition
    src=Idle_01
    via=AnimStateTransitionNode_0
   }
   {
    dst=ElimA_01
    rule=AnimGraph.AnimGraphNode_StateMachine_3.TESTING - Capture.AnimStateTransitionNode_1.Transition
    src=Idle_02
    via=AnimStateTransitionNode_1
   }
   {
    dst=ElimA_02
    rule=AnimGraph.AnimGraphNode_StateMachine_3.TESTING - Capture.AnimStateTransitionNode_6.Transition
    src=Idle_03
    via=AnimStateTransitionNode_6
   }
  ]
 }
 Idle_04 {
  path=AnimGraph.AnimGraphNode_StateMachine_3.TESTING - Capture.AnimStateNode_0.Idle_04
  nodes [
   n1 = Sequence Player 'EssayViewMyth_Idle_PACK'() {bLoopAnimation=false, sequence=&EssayViewMyth_Idle_PACK}
   Output Animation Pose(Result=@n1)
  ]
 }
 Idle_02 {
  path=AnimGraph.AnimGraphNode_StateMachine_3.TESTING - Capture.AnimStateNode_1.Idle_02
  nodes [
   n1 = Sequence Player 'EssayViewMyth_Idle_PACK'() {bLoopAnimation=false, sequence=&EssayViewMyth_Idle_PACK}
   Output Animation Pose(Result=@n1)
  ]
 }
 ElimA_01 {
  path=AnimGraph.AnimGraphNode_StateMachine_3.TESTING - Capture.AnimStateNode_2.ElimA_01
  nodes [
   n1 = Sequence Player 'EssayViewMyth_ElimA_PACK'() {bLoopAnimation=false, sequence=&EssayViewMyth_ElimA_PACK}
   Output Animation Pose(Result=@n1)
  ]
 }
 Elim_Out {
  path=AnimGraph.AnimGraphNode_StateMachine_3.TESTING - Capture.AnimStateNode_3.Elim_Out
  nodes [
   n1 = Sequence Player 'EssayViewMyth_Elim_Out_PACK'() {bLoopAnimation=false, sequence=&EssayViewMyth_Elim_Out_PACK}
   Output Animation Pose(Result=@n1)
  ]
 }
 ElimA_02 {
  path=AnimGraph.AnimGraphNode_StateMachine_3.TESTING - Capture.AnimStateNode_4.ElimA_02
  nodes [
   n1 = Sequence Player 'EssayViewMyth_ElimA_PACK'() {bLoopAnimation=false, sequence=&EssayViewMyth_ElimA_PACK}
   Output Animation Pose(Result=@n1)
  ]
 }
 ElimB {
  path=AnimGraph.AnimGraphNode_StateMachine_3.TESTING - Capture.AnimStateNode_5.ElimB
  nodes [
   n1 = Sequence Player 'EssayViewMyth_ElimB_PACK'() {bLoopAnimation=false, sequence=&EssayViewMyth_ElimB_PACK}
   Output Animation Pose(Result=@n1)
  ]
 }
 Idle_01 {
  path=AnimGraph.AnimGraphNode_StateMachine_3.TESTING - Capture.AnimStateNode_6.Idle_01
  nodes [
   n1 = Sequence Player 'EssayViewMyth_Idle_PACK'() {bLoopAnimation=false, sequence=&EssayViewMyth_Idle_PACK}
   Output Animation Pose(Result=@n1)
  ]
 }
 IdleBreak {path=AnimGraph.AnimGraphNode_StateMachine_3.TESTING - Capture.AnimStateNode_7.IdleBreak, nodes [Output Animation Pose()]}
 Idle_03 {
  path=AnimGraph.AnimGraphNode_StateMachine_3.TESTING - Capture.AnimStateNode_8.Idle_03
  nodes [
   n1 = Sequence Player 'EssayViewMyth_Idle_PACK'() {bLoopAnimation=false, sequence=&EssayViewMyth_Idle_PACK}
   Output Animation Pose(Result=@n1)
  ]
 }
 Transition {path=AnimGraph.AnimGraphNode_StateMachine_3.TESTING - Capture.AnimStateTransitionNode_0.Transition, nodes [Result()]}
 Transition {path=AnimGraph.AnimGraphNode_StateMachine_3.TESTING - Capture.AnimStateTransitionNode_1.Transition, nodes [Result()]}
 Transition {path=AnimGraph.AnimGraphNode_StateMachine_3.TESTING - Capture.AnimStateTransitionNode_2.Transition, nodes [Result()]}
 Transition {path=AnimGraph.AnimGraphNode_StateMachine_3.TESTING - Capture.AnimStateTransitionNode_3.Transition, nodes [Result()]}
 Transition {path=AnimGraph.AnimGraphNode_StateMachine_3.TESTING - Capture.AnimStateTransitionNode_4.Transition, nodes [Result()]}
 Transition {path=AnimGraph.AnimGraphNode_StateMachine_3.TESTING - Capture.AnimStateTransitionNode_5.Transition, nodes [Result()]}
 Transition {path=AnimGraph.AnimGraphNode_StateMachine_3.TESTING - Capture.AnimStateTransitionNode_6.Transition, nodes [Result()]}
 TESTING - WheelSpin {
  entry=Idle_01
  path=AnimGraph.AnimGraphNode_StateMachine_4.TESTING - WheelSpin
  ids=id
  nodes [
   AnimStateNode_0 = Idle_04()
   AnimStateNode_2 = ElimA_01() {
    stateEntered {guid=#15}
    stateFullyBlended {guid=#16}
    stateLeft {guid=#17}
   }
   AnimStateNode_3 = Elim_Out() {
    stateEntered {guid=#18}
    stateFullyBlended {guid=#19}
    stateLeft {guid=#20}
   }
   AnimStateNode_4 = ElimA_02() {
    stateEntered {guid=#21}
    stateFullyBlended {guid=#22}
    stateLeft {guid=#23}
   }
   AnimStateNode_5 = ElimB() {
    stateEntered {guid=#12}
    stateFullyBlended {guid=#13}
    stateLeft {guid=#14}
   }
   AnimStateNode_6 = Idle_01() {
    stateEntered {guid=#24}
    stateFullyBlended {guid=#25}
    stateLeft {guid=#26}
   }
   AnimStateNode_7 = IdleBreak()
   AnimStateNode_8 = Idle_03() {
    stateEntered {guid=#27}
    stateFullyBlended {guid=#28}
    stateLeft {guid=#29}
   }
   AnimStateTransitionNode_12 = AnimStateTransitionNode() {bAutomaticRuleBasedOnSequencePlayerInState=true, crossfadeDuration=0}
   AnimStateTransitionNode_13 = AnimStateTransitionNode() {bAutomaticRuleBasedOnSequencePlayerInState=true, crossfadeDuration=0}
   AnimStateTransitionNode_14 = AnimStateTransitionNode() {bAutomaticRuleBasedOnSequencePlayerInState=true, crossfadeDuration=0}
   AnimStateTransitionNode_2 = AnimStateTransitionNode() {bAutomaticRuleBasedOnSequencePlayerInState=true, crossfadeDuration=0}
   AnimStateTransitionNode_6 = AnimStateTransitionNode() {
    bAutomaticRuleBasedOnSequencePlayerInState=true
    crossfadeDuration=0
    transitionEnd {guid=#30}
    transitionInterrupt {guid=#31}
    transitionStart {guid=#32}
   }
   AnimStateTransitionNode_9 = AnimStateTransitionNode() {bAutomaticRuleBasedOnSequencePlayerInState=true, crossfadeDuration=0}
  ]
  links [
   {
    dst=Elim_Out
    rule=AnimGraph.AnimGraphNode_StateMachine_4.TESTING - WheelSpin.AnimStateTransitionNode_2.Transition
    src=ElimA_01
    via=AnimStateTransitionNode_2
   }
   {
    dst=ElimB
    rule=AnimGraph.AnimGraphNode_StateMachine_4.TESTING - WheelSpin.AnimStateTransitionNode_13.Transition
    src=ElimA_02
    via=AnimStateTransitionNode_13
   }
   {
    dst=Idle_01
    rule=AnimGraph.AnimGraphNode_StateMachine_4.TESTING - WheelSpin.AnimStateTransitionNode_14.Transition
    src=ElimB
    via=AnimStateTransitionNode_14
   }
   {
    dst=Idle_03
    rule=AnimGraph.AnimGraphNode_StateMachine_4.TESTING - WheelSpin.AnimStateTransitionNode_9.Transition
    src=Elim_Out
    via=AnimStateTransitionNode_9
   }
   {
    dst=ElimA_01
    rule=AnimGraph.AnimGraphNode_StateMachine_4.TESTING - WheelSpin.AnimStateTransitionNode_12.Transition
    src=Idle_01
    via=AnimStateTransitionNode_12
   }
   {
    dst=ElimA_02
    rule=AnimGraph.AnimGraphNode_StateMachine_4.TESTING - WheelSpin.AnimStateTransitionNode_6.Transition
    src=Idle_03
    via=AnimStateTransitionNode_6
   }
  ]
 }
 Idle_04 {
  path=AnimGraph.AnimGraphNode_StateMachine_4.TESTING - WheelSpin.AnimStateNode_0.Idle_04
  nodes [
   n1 = Sequence Player 'EssayViewMyth_Idle_PACK'() {bLoopAnimation=false, sequence=&EssayViewMyth_Idle_PACK}
   Output Animation Pose(Result=@n1)
  ]
 }
 ElimA_01 {
  path=AnimGraph.AnimGraphNode_StateMachine_4.TESTING - WheelSpin.AnimStateNode_2.ElimA_01
  nodes [
   n1 = Sequence Player 'EssayViewMyth_ElimA_PACK'() {bLoopAnimation=false, sequence=&EssayViewMyth_ElimA_PACK}
   Output Animation Pose(Result=@n1)
  ]
 }
 Elim_Out {
  path=AnimGraph.AnimGraphNode_StateMachine_4.TESTING - WheelSpin.AnimStateNode_3.Elim_Out
  nodes [
   n1 = Sequence Player 'EssayViewMyth_Elim_Out_PACK'() {bLoopAnimation=false, sequence=&EssayViewMyth_Elim_Out_PACK}
   Output Animation Pose(Result=@n1)
  ]
 }
 ElimA_02 {
  path=AnimGraph.AnimGraphNode_StateMachine_4.TESTING - WheelSpin.AnimStateNode_4.ElimA_02
  nodes [
   n1 = Sequence Player 'EssayViewMyth_ElimA_PACK'() {bLoopAnimation=false, sequence=&EssayViewMyth_ElimA_PACK}
   Output Animation Pose(Result=@n1)
  ]
 }
 ElimB {
  path=AnimGraph.AnimGraphNode_StateMachine_4.TESTING - WheelSpin.AnimStateNode_5.ElimB
  nodes [Output Animation Pose(Result=Sequence Player 'EssayViewMyth_ElimB_PACK'() {sequence=&EssayViewMyth_ElimB_PACK})]
 }
 Idle_01 {
  path=AnimGraph.AnimGraphNode_StateMachine_4.TESTING - WheelSpin.AnimStateNode_6.Idle_01
  nodes [
   n1 = Sequence Player 'EssayViewMyth_Idle_PACK'() {bLoopAnimation=false, sequence=&EssayViewMyth_Idle_PACK}
   Output Animation Pose(Result=@n1)
  ]
 }
 IdleBreak {path=AnimGraph.AnimGraphNode_StateMachine_4.TESTING - WheelSpin.AnimStateNode_7.IdleBreak, nodes [Output Animation Pose()]}
 Idle_03 {
  path=AnimGraph.AnimGraphNode_StateMachine_4.TESTING - WheelSpin.AnimStateNode_8.Idle_03
  nodes [
   n1 = Sequence Player 'EssayViewMyth_Idle_PACK'() {bLoopAnimation=false, sequence=&EssayViewMyth_Idle_PACK}
   Output Animation Pose(Result=@n1)
  ]
 }
 Transition {path=AnimGraph.AnimGraphNode_StateMachine_4.TESTING - WheelSpin.AnimStateTransitionNode_12.Transition, nodes [Result()]}
 Transition {path=AnimGraph.AnimGraphNode_StateMachine_4.TESTING - WheelSpin.AnimStateTransitionNode_13.Transition, nodes [Result()]}
 Transition {path=AnimGraph.AnimGraphNode_StateMachine_4.TESTING - WheelSpin.AnimStateTransitionNode_14.Transition, nodes [Result()]}
 Transition {path=AnimGraph.AnimGraphNode_StateMachine_4.TESTING - WheelSpin.AnimStateTransitionNode_2.Transition, nodes [Result()]}
 Transition {path=AnimGraph.AnimGraphNode_StateMachine_4.TESTING - WheelSpin.AnimStateTransitionNode_6.Transition, nodes [Result()]}
 Transition {path=AnimGraph.AnimGraphNode_StateMachine_4.TESTING - WheelSpin.AnimStateTransitionNode_9.Transition, nodes [Result()]}
 EventGraph {nodes [
   Event OnCharacterCustomizationLoaded()
   > n1 = Cast To FortPlayerPawn(Object=TryGetFortPawn())
   > n2 = Cast To B_Athena_PartModifier_Backpack_ChiveFlake!(
    execute
    Object=GetModifierForPartType!(self:Fort Player Pawn Object Reference=@n1, PartType:EFortCustomPartType Enum=Backpack) -> (ReturnValue:Custom Character Part Modifier Object Reference)
   ) -> (then, CastFailed, AsB Athena Part Modifier Backpack Chive Flake:B Athena Part Modifier Backpack Chive Flake Object Reference) {is=title K2Node_DynamicCast}
   > n3 = Branch(Condition=Get FE_ReactivityEnabled!(self:B Athena Part Modifier Backpack Chive Flake Object Reference=@n2) -> (FE_ReactivityEnabled:bool))
   > Do Once(Reset=@n5.then)
   > n4 = Delay(Duration=4)
   Set Play Reactive FE Anim(execute=[@n7.then, @n4.then], Play Reactive FE Anim:bool=true) -> (Output_Get:bool)
   n5 = Set Play Reactive FE Anim(execute=@n3.else, Play Reactive FE Anim=false)
   n6 = Event BlueprintProcessAnimInputEvent()
   > Branch(Condition=IsNamed(self=@n6.AnimInputEvent, InEventName=aie.EnemyPlayerEliminated))
   > Set bEnemyPlayerEliminated(bEnemyPlayerEliminated:bool=true) -> (Output_Get:bool)
   > Delay(Duration=0.2)
   > Set bEnemyPlayerEliminated(bEnemyPlayerEliminated=false)
   Event BlueprintPostEvaluateAnimation() {is=title K2Node_Event}
   > Set bEnemyPlayerEliminated(bEnemyPlayerEliminated=false)
   > Set Play Reactive FE Anim(Play Reactive FE Anim=false)
   AnimNotify_ResetReactivity() {is=title K2Node_Event}
   > Set Play Reactive FE Anim(Play Reactive FE Anim=false)
   > n7 = Delay(Duration=10)
   n8 = Event BlueprintUpdateAnimation()
   > n9 = Sequence!(execute) -> (then_0, then_1, then_2:Exec, then_3:Exec)
   n10 = GetOwnerSkeletalMeshComponent()
   n11 = BreakRotator(InRot=BreakTransform(InTransform=GetDeltaTransformFromRefPose(self=@n10, BoneName=spine_05, BaseName=spine_01)).Rotation)
   Set backpack_rotate(execute=@n9.then_2, backpack_rotate=MakeRotator(Roll=MapRangeClamped(Value=@n11.Yaw, InRangeA=0, InRangeB=-30, OutRangeB=45)))
   > Set backpack_translation(backpack_translation=MakeVector(Y=MapRangeClamped(Value=@n11.Yaw, InRangeA=0, InRangeB=-30, OutRangeA=-1.2, OutRangeB=-3)))
   Branch(execute=@n9.then_3, Condition=Get isFETriggerFX() -> (isFETriggerFX:bool))
   > Call FETrigger FX!(execute:Exec, self:M MED Essay View Myth Pack Anim BP Object Reference) -> (then:Exec) {is=title K2Node_CallDelegate}
   > Set isFETriggerFX(isFETriggerFX:bool=false) -> (Output_Get:bool)
   Set AnimCurve(
    execute=@n9.then_1
    AnimCurve:double=Remap Float(Input=GetCurveValue(CurveName=DynamicsCurve), OldRange=(X=0,Y=1), NewRange=(X=1,Y=0)) -> (Output=~)
   ) -> (Output_Get:double)
   > Set Delta Time X(Delta Time X:double=@n8.DeltaTimeX) -> (Output_Get:double)
   n12 = float * float(A=GetCurveValue(CurveName=WheelSpinRate), B=Get Delta Time X() -> (Delta Time X:double)) {is=title K2Node_PromotableOperator}
   n13 = Get Wheel Rotation Last Tick Value() -> (Wheel Rotation Last Tick Value:double)
   > n14 = Set Wheel Rotation Value(Wheel Rotation Value:double="% (Float)"!(A:double=float + float(A=@n12, B=@n13), B:double=360) -> (ReturnValue:double)) -> (Output_Get:double)
   > Set Wheel Rotation(Wheel Rotation:Rotator=MakeRotator(Pitch=@n14)) -> (Output_Get:Rotator)
   > Set Wheel Rotation Last Tick Value(Wheel Rotation Last Tick Value:double=Get Wheel Rotation Value() -> (Wheel Rotation Value:double)) -> (Output_Get:double)
   n15 = Gravity Override State Function DEPRECATED(
    execute=@n9.then_0
    CharacterPartAnimInstance=Self-Reference()
    GameTimeInSeconds=GetGameTimeInSeconds()
    GravityOverrideParams=Get Gravity Override Params()
   )
   > Set GravityOverride(GravityOverride=@n15)
   Branch(Condition=Get FrontEnd() -> (FrontEnd:bool))
   > Set Play Reactive FE Anim(Play Reactive FE Anim=true)
   n16 = Branch(Condition=IsCosmeticObjectDisplayedInFrontendContext!(CosmeticObject:Object Reference=Self-Reference()) -> (ReturnValue:bool))
   > Set FrontEnd(FrontEnd:bool=true) -> (Output_Get:bool)
   Set FrontEnd(execute=@n16.else, FrontEnd=false)
   TryGetPawnOwner()
  ]}
 FETriggerFX {nodes [FETriggerFX() {is=title K2Node_FunctionEntry}]}
 OnStateEntry {nodes [
   OnStateEntry!() -> (then, Context:"Anim Update Context Structure (by ref)", Node:"Anim Node Reference Structure (by ref)") {is=title K2Node_FunctionEntry}
   > Set isFETriggerFX(isFETriggerFX=true)
   > Return Node()
  ]}
]
variables [
 AnimCurve {type=number}
 backpack_rotate {type=Rotator}
 backpack_translation {type=Vector}
 bEnemyPlayerEliminated {type=boolean}
 Delta Time X {type=number}
 FETriggerFX {default=()}
 FrontEnd {type=boolean}
 Gravity Override Params {type=GravityOverrideParamsStruct}
 GravityOverride {type=Vector}
 isFETriggerFX {type=boolean}
 Play Reactive FE Anim {type=boolean}
 Wheel Rotation {type=Rotator}
 Wheel Rotation Last Tick Value {type=number, default=1}
 Wheel Rotation Rate Smoothed {type=number}
 Wheel Rotation Value {type=number, default=1}
 WheelSpin_Anim_PlayRate {type=number, default=1}
]
